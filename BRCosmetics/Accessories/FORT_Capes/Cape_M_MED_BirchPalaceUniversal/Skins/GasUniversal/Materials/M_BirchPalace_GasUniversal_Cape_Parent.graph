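# Parent material for the BirchPalace GasUniversal cape, a cloth-shaded cosmetic built on the 2023 anime character shader.
# The shader's Pre, Calculations and Post functions run in sequence, with lit and shaded colors blended before Post.
# An emissive mask's red and green channels define pattern glow masks, alongside tunable base and core glow colors and panning noise breakup.
# A borrowed glider flowing-smoke function, scaled by pulse strength, adds to the emissive output.
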Material /BRCosmetics/Accessories/FORT_Capes/Cape_M_MED_BirchPalaceUniversal/Skins/GasUniversal/Materials/M_BirchPalace_GasUniversal_Cape_Parent
graph {
 properties {
  bAutomaticallySetUsageInEditor=false
  bUseMaterialAttributes=true
  bUsedWithClothing=true
  bUsedWithMeshDeformer=true
  bUsedWithMorphTargets=true
  bUsedWithSkeletalMesh=true
  pixelDepthOffsetMode=PDOM_Legacy
  shadingModel=MSM_Cloth
  usedShadingModels=Cloth
 }
 nodes [
  n1 = MaterialFunctionCall!(CustomUV) {materialFunction=/Game/Global/Materials/Character_Parent_2023/Anime_Parent_2023/MaterialFunctions/MFC_AnimeShader_2023_Pre}
  n2 = MaterialFunctionCall!(CustomLightVector, OverrideNormal_WS, STT=@n1.STT, Normal_Map=@n1.STT, DFL=@n1.STT, FX=@n1.STT) {materialFunction=/Game/Global/Materials/Character_Parent_2023/Anime_Parent_2023/MaterialFunctions/MFC_AnimeShader_2023_Calculations}
  n3 = NamedRerouteUsage()
  n4 = LinearInterpolate(A=@n3, B=NamedRerouteUsage())
  n5 = NamedRerouteUsage()
  MaterialFunctionCall!(
   Shading_Gradient=@n2.Shading_Gradient
   Threshold_Gradient=@n2.Shading_Gradient
   Specular_Gradient=@n2.Shading_Gradient
   Sheen_Mask=@n2.Shading_Gradient
   Thickness=@n2.Shading_Gradient
   DFL_Mask=@n2.Shading_Gradient
   FX=@n2.Shading_Gradient
   Metallic_Gradient=@n2.Shading_Gradient
   RimLight_Gradient=@n2.Shading_Gradient
   NormalBuffer=@n2.Shading_Gradient
   --------
   CINE_LightSources=@n2.Shading_Gradient
   CINE_LightColor=@n2.Shading_Gradient
   ---------
   Color_Lit=LinearInterpolate(A=@n3, B=LinearInterpolate(A=@n3, B=@n4, Alpha=@n5), Alpha=NamedRerouteUsage())
   Color_Shaded=LinearInterpolate(A=NamedRerouteUsage(), B=NamedRerouteUsage(), Alpha=NamedRerouteUsage())
   SRM=@n1.SRM
   Emissive=NamedRerouteUsage()
   DFL_Color=@n1.SRM
   Specular_Color=@n1.SRM
   RimLight_Color=@n1.SRM
  ) {materialFunction=/Game/Global/Materials/Character_Parent_2023/Anime_Parent_2023/MaterialFunctions/MFC_AnimeShader_2023_Post}
  NamedRerouteDeclaration(None=@n1.Color_Lit) {name=IN_Color_Lit, nodeColor=<rgba 0 1 0.4141 1>}
  NamedRerouteDeclaration(None=@n1.Color_Shaded) {name=IN_Color_Shaded, nodeColor=<rgba 1 0.0938 0 1>}
  n6 = TextureObjectParameter() {
   group=BirchPalace
   parameterName=EmissiveMask
   texture=/BRCosmetics/Accessories/FORT_Capes/Cape_M_MED_BirchPalaceUniversal/Skins/GasUniversal/Textures/T_BirchPalace_GasUniversal_Cape_E
  }
  NamedRerouteDeclaration(None=@n6) {name=TA_Mask, nodeColor=<rgba 1 0 0.2109 1>}
  n7 = NamedRerouteDeclaration(None=VectorParameter() {group=BirchPalace, parameterName=Glow_BaseColor, defaultValue=<rgba 0.0375 0.0587 0.25 1>}.RGB) {name=EmissiveColor, nodeColor=<rgba 0.875 0 1 1>}
  n8 = NamedRerouteDeclaration(None=VectorParameter() {group=BirchPalace, parameterName=Glow_CoreColor, defaultValue=<rgba 0.5652 2.4658 5 1>}.RGB) {name=Glow_CoreColor, nodeColor=<rgba 0 1 0.5078 1>}
  n9 = Multiply(
   A=Multiply(A=NamedRerouteUsage(), B=NamedRerouteUsage())
   B=LinearInterpolate(A=Constant() {r=0.8}, B=Constant() {r=1.3}, Alpha=NamedRerouteUsage()) {constA=1, constB=3}
  )
  n10 = Multiply(A=Multiply(A=LinearInterpolate(A=@n7, B=@n8, Alpha=@n9), B=@n9), B=NamedRerouteUsage())
  n11 = LinearInterpolate(Alpha=NamedRerouteUsage()) {constA=1, constB=0.05}
  n12 = MaterialFunctionCall!(In_FireSmoke_TAMask=NamedRerouteUsage(), In_FireSmoke_UV) {materialFunction=/BRCosmetics/Gadgets/Assets/VinderTech_GliderChute/Glider_RampCanal/Materials/MFW_FlowingSmoke}
  n13 = Multiply(A=ScalarParameter() {defaultValue=0.15, group=BirchPalace, parameterName=SmokePulseStrength}, B=NamedRerouteUsage())
  NamedRerouteDeclaration(None=Add(A=Multiply(A=@n10, B=@n11), B=Multiply(A=@n12.Result, B=@n13) {constB=0.1})) {name=OutEmissive, nodeColor=<rgba 0.8281 0 1 1>}
  NamedRerouteDeclaration(None=@n1.DFL) {name=In_DFL, nodeColor=<rgba 0.1953 1 0 1>}
  n14 = TextureCoordinate()
  n15 = AppendVector(
   A=ScalarParameter() {defaultValue=0.8, group=BirchPalace, parameterName=GlowNoise_TilingX}
   B=ScalarParameter() {defaultValue=2.5, group=BirchPalace, parameterName=GlowNoise_TilingY}
  )
  n16 = ScalarParameter() {defaultValue=0.01, group=BirchPalace, parameterName=GlowNoise_SpeedX}
  n17 = ScalarParameter() {defaultValue=0.3, group=BirchPalace, parameterName=GlowNoise_SpeedY}
  n18 = TextureSample(UVs=Panner(Coordinate=Multiply(A=@n14, B=@n15) {constB=11}, Speed=AppendVector(A=@n16, B=@n17)) {speedY=0.12}) {
   samplerSource=SSM_Wrap_WorldGroupSettings
   samplerType=SAMPLERTYPE_Grayscale
   texture=/Game/Global/Textures/Default/Noise/LowResBlurredNoiseNormalized
  }
  NamedRerouteDeclaration(None=LinearInterpolate(
    A=ScalarParameter() {defaultValue=0.2, group=BirchPalace, parameterName=GlowNoiseIntensity_Min}
    B=ScalarParameter() {defaultValue=0.8, group=BirchPalace, parameterName=GlowNoiseIntensity_Max}
    Alpha=@n18.R
   ) {constA=0.2, constB=0.8}) {name=GlowNoise_Breakup_01, nodeColor=<rgba 0 1 0.7891 1>}
  n19 = TextureSample(UVs=TextureCoordinate(), Tex=@n6)
  NamedRerouteDeclaration(None=@n19.R) {name=PatternGlowMask, nodeColor=<rgba 0.6641 0 1 1>}
  NamedRerouteDeclaration(None=@n19.G) {name=PatternGlowDetailMask, nodeColor=<rgba 0.6641 0 1 1>}
 ]
}
